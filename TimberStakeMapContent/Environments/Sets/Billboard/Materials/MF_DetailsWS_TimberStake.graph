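# Material function that layers triplanar world-space detail onto input material attributes, defaulting to TimberStake billboard wall textures.
# One packed texture drives base color, specular and roughness variation (R) and normals (GB), sampled at separate tilings.
# Four weighted slots picked by UV1.x mask the blend, behind a static switch; the interior-wall mask is left unconnected.
# Low quality uses weaker normals and subtracts a directional dot-product term from base color.
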MaterialFunction /TimberStakeMapContent/Environments/Sets/Billboard/Materials/MF_DetailsWS_TimberStake
graph {nodes [
  n1 = Reroute() {desc=BaseColor}
  n2 = Multiply(B=VectorParameter() {group=DetailsWS, parameterName=WSD_Basecolor_Mult, defaultValue=<rgba 1 1 1 1>}.RGB)
  n3 = Multiply(A=@n2, B=@n1)
  n4 = Step(Y=@n1) {constX=0.5}
  n5 = Multiply(A=Multiply(A=OneMinus(None=@n2), B=OneMinus(None=@n1))) {constB=2}
  Add(A=Multiply(A=Multiply(A=@n3) {constB=2}, B=@n4), B=Multiply(A=OneMinus(None=@n5), B=OneMinus(None=@n4)))
  n6 = NamedRerouteUsage()
  n7 = TextureSample() {samplerType=SAMPLERTYPE_Masks, texture=/TimberStakeMapContent/Environments/Sets/Billboard/Textures/T_Billboard_Exterior_S}
  n8 = SetMaterialAttributes!(
   MaterialAttributes
   BaseColor=Constant3Vector() {constant=<rgba 0.2542 0.5395 0.7231 1>}
   Metallic=@n7.G
   Specular=@n7.G
   Roughness=@n7.G
   Normal
  ) {attributeSetTypes [
    #1
    #2
    #3
    #4
    #5
   ]}
  n9 = GetMaterialAttributes(None=NamedRerouteDeclaration(None=FunctionInput(Preview=@n8) {inputType=FunctionInput_MaterialAttributes}) {name=IN_MA, nodeColor=<rgba 0 1 0.4375 1>}) {attributeGetTypes [
    #1
    #2
    #3
    #4
    #5
   ]}
  n10 = SetMaterialAttributes(
   MaterialAttributes=@n9.MaterialAttributes
   BaseColor=NamedRerouteUsage()
   Specular=NamedRerouteUsage()
   Roughness=NamedRerouteUsage()
   Normal=NamedRerouteUsage()
  ) {attributeSetTypes [
    #1
    #2
    #3
    #4
    #5
   ]}
  n11 = Add(A=ComponentMask(None=TextureCoordinate() {coordinateIndex=1}) {r=true}) {constB=-0.5}
  n12 = Abs(None=@n11)
  n13 = OneMinus(None=Saturate(None=Round(None=@n12)))
  n14 = ScalarParameter() {defaultValue=1, group=DetailsWS, parameterName=WSD L1, sortPriority=-4}
  n15 = Add(A=@n11) {constB=-1}
  n16 = Abs(None=@n15)
  n17 = OneMinus(None=Saturate(None=Round(None=@n16)))
  n18 = ScalarParameter() {defaultValue=1, group=DetailsWS, parameterName=WSD L2, sortPriority=-3}
  n19 = Add(A=@n15) {constB=-1}
  n20 = Abs(None=@n19)
  n21 = OneMinus(None=Saturate(None=Round(None=@n20)))
  n22 = ScalarParameter() {defaultValue=1, group=DetailsWS, parameterName=WSD L3, sortPriority=-2}
  n23 = Abs(None=Add(A=@n19) {constB=-1})
  n24 = OneMinus(None=Saturate(None=Round(None=@n23)))
  n25 = ScalarParameter() {defaultValue=1, group=DetailsWS, parameterName=WSD L4, sortPriority=-1}
  n26 = Max(A=Max(A=Multiply(A=@n13, B=@n14), B=Multiply(A=@n17, B=@n18)), B=Max(A=Multiply(A=@n21, B=@n22), B=Multiply(A=@n24, B=@n25)))
  FunctionOutput(None=StaticSwitchParameter(True=BlendMaterialAttributes(A=@n6, B=@n10, Alpha=@n26), False=NamedRerouteUsage()) {group=DetailsWS, parameterName=UseDetailsWS, sortPriority=-100})
  n27 = Multiply(B=VectorParameter() {group=DetailsWS, parameterName=WSD_Basecolor_Mult, defaultValue=<rgba 1 1 1 1>}.RGB)
  n28 = NamedRerouteUsage()
  n29 = Multiply(A=OneMinus(None=@n27), B=OneMinus(None=@n28))
  LinearInterpolate(A=OneMinus(None=Multiply(A=@n29) {constB=2}), B=Multiply(A=Multiply(A=@n27, B=@n28)) {constB=2}, Alpha=Step(Y=@n28) {constX=0.5})
  MaterialFunctionCall() {materialFunction=&CameraDirectionVector}
  n30 = Multiply(A=LocalPosition() {localOrigin=InstancePreSkinning}.XYZ) {constB=0.1}
  n31 = ScalarParameter() {
   defaultValue=1
   desc="Mask based on mesh local Y. Will make what is on negative side from pivot free of effect. In some cases you can't use it ( Curved walls) so instead use additional material layer to controll dirt level inside buildings"
   group=DetailsWS
   parameterName=WSD_Interior Walls
   sortPriority=0
  }
  n32 = Max(A=Saturate(None=ComponentMask(None=@n30) {g=true}), B=@n31)
  Multiply(A=@n26, B=ShaderStageSwitch(PixelShader=VertexInterpolator(VS=@n32).PS, VertexShader=@n32))
  n33 = ScalarParameter() {defaultValue=0.125, group=DetailsWS, parameterName=WSD_BaseColor, sliderMax=1}
  n34 = Add(A=@n33) {constB=-0.5}
  n35 = Subtract(B=@n34) {constA=0, desc=Low negative value}
  n36 = Add(A=@n34) {desc=High Positive}
  n37 = TextureSample(
   UVs=MaterialFunctionCall!(UV Scale=ScalarParameter() {defaultValue=800, group=DetailsWS, parameterName=WSD_Tile, sortPriority=0}, Seam Sharpness) {materialFunction=/Game/Athena/Environments/MaterialFunctions/MF_TriplanarFlat}.UVs
   Tex=NamedRerouteUsage()
  ) {samplerType=SAMPLERTYPE_LinearColor}
  Multiply(A=NamedRerouteUsage(), B=Multiply(A=LinearInterpolate(A=@n35, B=@n36, Alpha=@n37.R) {desc=Lerps between greater range of value}) {constB=2})
  NamedRerouteDeclaration(None=@n9.Normal) {name=IN_Normal, nodeColor=<rgba 1 0 0.7266 1>}
  n38 = Add(A=@n37.R) {constB=-0.5}
  n39 = Add(A=NamedRerouteUsage(), B=Multiply(A=@n38, B=@n33))
  n40 = DotProduct(A=NamedRerouteUsage(), B=Normalize(VectorInput=Constant3Vector() {constant=<rgba 0 1 0.1 0>}))
  n41 = ScalarParameter() {defaultValue=-0.08, group=DetailsWS, parameterName=WSD_BaseColor_LowQ_DetailsBias, sliderMax=1}
  n42 = Saturate(None=Multiply(A=Add(A=@n40, B=@n41), B=ScalarParameter() {defaultValue=0.55, group=DetailsWS, parameterName=WSD_BaseColor_LowQ_Details, sliderMax=1}))
  NamedRerouteDeclaration(None=QualitySwitch(Default=@n39, Low=Subtract(A=@n39, B=@n42))) {name=OUT_Basecolor, nodeColor=<rgba 1 0 0.2578 1>}
  n43 = MaterialFunctionCall(UV Scale=ScalarParameter() {defaultValue=64, group=DetailsWS, parameterName=WSD_Normal_Tile, sortPriority=10}) {materialFunction=&MF_TriplanarFlat}
  n44 = ComponentMask(None=TextureSample(UVs=@n43.UVs, Tex=NamedRerouteUsage()) {samplerType=SAMPLERTYPE_LinearColor}.RGBA) {b=true, g=true}
  n45 = ScalarParameter() {defaultValue=0.8, group=DetailsWS, parameterName=WSD_Normal, sortPriority=10}
  n46 = Multiply(A=@n45, B=ScalarParameter() {defaultValue=0.5, group=DetailsWS, parameterName=WSD_Normal_LowQ_Mult, sortPriority=10})
  n47 = DeriveNormalZ(InXY=Multiply(A=ConstantBiasScale(None=@n44) {bias=-0.5, scale=2}, B=QualitySwitch(Default=@n45, Low=@n46)))
  NamedRerouteDeclaration(None=MaterialFunctionCall!(
    Tangent Normals=NamedRerouteDeclaration(None=@n47) {name=WS_Normals_Tangent, nodeColor=<rgba 0 1 0.5547 1>}
    Vertex Normals=NamedRerouteUsage()
    Triplanar Mask=@n43.Triplanar Mask
   ) {materialFunction=/Game/Athena/Environments/MaterialFunctions/MF_TriplanarNormalTangentToWorld}.World Normals) {name=OUT_Normal, nodeColor=<rgba 1 0.1406 0 1>}
  n48 = NamedRerouteUsage()
  n49 = Add(A=@n38, B=ScalarParameter() {defaultValue=-1, group=DetailsWS, parameterName=WSD_Specular_Expand, sliderMax=1})
  n50 = ScalarParameter() {defaultValue=0.45, group=DetailsWS, parameterName=WSD_Specular_Contrast, sliderMax=1}
  n51 = Add(A=@n48, B=Multiply(A=@n48, B=Multiply(A=@n49, B=@n50)))
  NamedRerouteDeclaration(None=@n51) {name=OUT_Specular, nodeColor=<rgba 0 0.1797 1 1>}
  n52 = NamedRerouteUsage()
  n53 = Add(A=Multiply(A=@n38) {constB=-1}, B=ScalarParameter() {defaultValue=1.5, group=DetailsWS, parameterName=WSD_Roughness_Expand, sliderMax=1}) {constB=-0.5}
  n54 = ScalarParameter() {defaultValue=0.8, group=DetailsWS, parameterName=WSD_Roughness_Contrast, sliderMax=1}
  n55 = Add(A=@n52, B=Multiply(A=@n52, B=Multiply(A=@n53, B=@n54)))
  NamedRerouteDeclaration(None=@n55) {name=OUT_Roughness, nodeColor=<rgba 0.5 1 0 1>}
  n56 = TextureObjectParameter() {
   desc="R= BaseColor, Spec, Rough\r\nGB= Normal RG"
   group=DetailsWS
   parameterName=DetailsTexture
   samplerType=SAMPLERTYPE_LinearColor
   sortPriority=0
   texture=/TimberStakeMapContent/Environments/Sets/Billboard/Textures/T_TimberStake_WallDetails
  }
  NamedRerouteDeclaration(None=@n56) {name=IN_Detail_T, nodeColor=<rgba 1 0 0.8203 1>}
  NamedRerouteDeclaration(None=@n9.BaseColor) {name=IN_BaseColor, nodeColor=<rgba 1 0 0.7266 1>}
  NamedRerouteDeclaration(None=@n9.Metallic) {name=IN_Metallic, nodeColor=<rgba 1 0 0.7266 1>}
  NamedRerouteDeclaration(None=@n9.Specular) {name=IN_Specular, nodeColor=<rgba 1 0 0.7266 1>}
  NamedRerouteDeclaration(None=@n9.Roughness) {name=IN_Roughness, nodeColor=<rgba 1 0 0.7266 1>}
  Step(X=@n51) {constX=0.46, constY=0.46}
  Step(X=@n55) {constY=0.25}
  VertexNormalWS()
 ]}
